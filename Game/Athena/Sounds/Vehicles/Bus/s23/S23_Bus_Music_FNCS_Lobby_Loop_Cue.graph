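# Sound cue for the Season 23 FNCS Battle Bus lobby music loop.
# A looping wave player for s23_FNCS_BattleBus_Loop feeds a root Enveloper, whose volume curve rises from zero to full over the first second as a fade-in.
# It borrows the close-range lobby bus engine attenuation, sits in the Battle Bus sound class and plays slightly below full volume.
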SoundCue /Game/Athena/Sounds/Vehicles/Bus/s23/S23_Bus_Music_FNCS_Lobby_Loop_Cue
root=Enveloper_0
nodes [
 {
  id=Enveloper_0
  children [WavePlayer_0]
  properties {pitchCurve {editorCurveData {keys [{time=0}]}}, volumeCurve {editorCurveData {keys [<at 0 0>,<at 1 1>]}}}
 }
 {id=WavePlayer_0, properties {bLooping=true, soundWaveAssetPtr=&/s23_FNCS_BattleBus_Loop}}
]
properties {
 attenuationSettings=/Game/Sounds/Attenuation/Fort_Bus_Lobby_Engine_Athena_Close_Att
 maxDistance=12500
 soundClassObject=/Game/Sounds/FortSoundClassesAndModes/Fort_BattleBus
 volumeMultiplier=0.9
}
