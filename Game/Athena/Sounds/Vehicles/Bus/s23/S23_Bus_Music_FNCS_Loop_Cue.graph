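# Sound cue for the Battle Bus music loop themed for FNCS, from the s23 bus sound set in Athena.
# A single looping Wave Player plays the s23 FNCS Battle Bus loop wave directly as the root, with slightly reduced volume.
# The cue routes to the Battle Bus sound class and references the shared Athena bus music attenuation asset.
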SoundCue /Game/Athena/Sounds/Vehicles/Bus/s23/S23_Bus_Music_FNCS_Loop_Cue
root=WavePlayer_1
nodes [{id=WavePlayer_1, properties {bLooping=true, soundWaveAssetPtr=&/s23_FNCS_BattleBus_Loop}}]
properties {
 attenuationSettings=/Game/Sounds/Attenuation/Fort_Bus_Music_Athena_Att
 maxDistance=21000
 soundClassObject=/Game/Sounds/FortSoundClassesAndModes/Fort_BattleBus
 volumeMultiplier=0.9
 attenuationOverrides {bAttenuate=false, bSpatialize=false}
}
